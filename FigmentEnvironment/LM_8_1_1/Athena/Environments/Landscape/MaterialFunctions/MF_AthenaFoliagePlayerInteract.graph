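# Material function that supplies a bend vector and proximity mask for foliage reacting to the player.
# The normalized direction from world position to PlayerPosition is flattened vertically and scaled by BendStrength, whose negative default points it away from the player, giving the Rotation output.
# A soft, clamped sphere mask around the player, sized by PlayerRadiusMax, forms PlayerSphereMask.
# The two outputs stay separate, leaving their combination to the calling material.
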MaterialFunction /FigmentEnvironment/LM_8_1_1/Athena/Environments/Landscape/MaterialFunctions/MF_AthenaFoliagePlayerInteract
graph {nodes [
  n1 = ComponentMask(None=FunctionInput(Preview=Constant4Vector().RGBA) {bUsePreviewValueAsDefault=true, inputName=PlayerPosition, inputType=FunctionInput_Vector4}) {b=true, g=true, r=true}
  n2 = WorldPosition()
  n3 = Normalize(VectorInput=Subtract(A=@n1, B=@n2.XYZ))
  n4 = Constant3Vector() {constant=<rgba 1 1 0.2 0>}
  n5 = Constant() {r=-0.07}
  FunctionOutput(None=Multiply(
    A=Multiply(A=@n3, B=@n4)
    B=FunctionInput(Preview=@n5) {bUsePreviewValueAsDefault=true, inputName=BendStrength, inputType=FunctionInput_Scalar}
   )) {outputName=Rotation}
  n6 = FunctionInput(Preview=Constant() {r=128}) {bUsePreviewValueAsDefault=true, inputName=PlayerRadiusMax, inputType=FunctionInput_Scalar}
  FunctionOutput(None=Clamp(None=SphereMask(A=@n1, B=@n2.XYZ, Radius=@n6) {attenuationRadius=64, hardnessPercent=20})) {outputName=PlayerSphereMask}
 ]}
