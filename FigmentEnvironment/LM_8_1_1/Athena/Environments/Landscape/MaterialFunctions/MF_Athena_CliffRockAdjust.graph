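# Material function that adjusts an incoming cliff material for the Athena landscape, tinting its base color with a terrain rock color map.
# The map is sampled at world position through MF_Athena_Terrain_Coords, scaled by ColormapDivisor from MPC_Landscape, then multiplied into BaseColor.
# Roughness is replaced by the RoughRock parameter.
# The other connected outputs, including Normal, are all wired from the input's Metallic as written.
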MaterialFunction /FigmentEnvironment/LM_8_1_1/Athena/Environments/Landscape/MaterialFunctions/MF_Athena_CliffRockAdjust
graph {nodes [
  Constant()
  n1 = BreakMaterialAttributes!(Attr=FunctionInput(Preview=MakeMaterialAttributes!(
     BaseColor
     Metallic
     Specular
     Roughness
     Anisotropy
     EmissiveColor
     Opacity
     OpacityMask
     Normal
     Tangent
     WorldPositionOffset
     SubsurfaceColor
     ClearCoat
     ClearCoatRoughness
     AmbientOcclusion
     Refraction
     CustomizedUVs_0
     CustomizedUVs_1
     CustomizedUVs_2
     CustomizedUVs_3
     CustomizedUVs_4
     CustomizedUVs_5
     CustomizedUVs_6
     CustomizedUVs_7
     PixelDepthOffset
     ShadingModel
     Displacement
    )) {inputName=CliffMaterial, inputType=FunctionInput_MaterialAttributes})
  n2 = TextureSample(UVs=MaterialFunctionCall!(WorldPosition) {materialFunction=/FigmentEnvironment/LM_8_1_1/Athena/Environments/Landscape/MaterialFunctions/Utilities/MF_Athena_Terrain_Coords}.Result) {
   samplerSource=SSM_Wrap_WorldGroupSettings
   samplerType=SAMPLERTYPE_LinearColor
   texture=/FigmentEnvironment/LM_8_1_1/Athena/Environments/Landscape/Textures/T_Athena_Terrain_RockColor
  }
  n3 = Multiply(
   A=@n2.RGB
   B=CollectionParameter() {collection=/FigmentEnvironment/LM_8_1_1/Athena/Environments/Landscape/MPC/MPC_Landscape, parameterName=ColormapDivisor}
  )
  FunctionOutput(None=MakeMaterialAttributes(
    BaseColor=Multiply(A=@n1.BaseColor, B=@n3)
    Metallic=@n1.Metallic
    Roughness=ScalarParameter() {defaultValue=0.9, parameterName=RoughRock}
    EmissiveColor=@n1.Metallic
    Opacity=@n1.Metallic
    OpacityMask=@n1.Metallic
    Normal=@n1.Metallic
    WorldPositionOffset=@n1.Metallic
    SubsurfaceColor=@n1.Metallic
    ClearCoat=@n1.Metallic
    ClearCoatRoughness=@n1.Metallic
    AmbientOcclusion=@n1.Metallic
    Refraction=@n1.Metallic
    PixelDepthOffset=@n1.Metallic
   ))
 ]}
