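# Material function that colors Athena landscape flower foliage by biome from a packed mask texture, outputting color and alpha.
# Green drives color gradient maps, blue blends center colors, red selects terrain-matched or lake-grass stalk color.
# Rounded biome masks choose Farm, Swamp or Mountain palettes, which the instance random alpha blends against the default.
# Rounded element randomness scales gradient range, or grayscale for Farm; a static switch can force one base color.
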MaterialFunction /FigmentEnvironment/LM_8_1_1/Athena/Environments/Landscape/MaterialFunctions/MF_Athena_FlowerColors
graph {nodes [
  n1 = TextureSample(Tex=FunctionInput(Preview=TextureObject() {samplerType=SAMPLERTYPE_LinearColor, texture=/FigmentEnvironment/LM_8_1_1/Athena/Environments/Landscape/Foliage/T_Athena_Flowers_01}) {bUsePreviewValueAsDefault=true, inputType=FunctionInput_Texture2D}) {samplerType=SAMPLERTYPE_LinearColor}
  n2 = FunctionInput() {
   bUsePreviewValueAsDefault=true
   description=Rounded to nearest 0-1. Best used with RandomValuePerElement from PerObjectFoliageData to Lerp between 2 different color sets.
   inputName=Element_RandomAlpha
   inputType=FunctionInput_Scalar
   sortPriority=1
   previewValue=<1 1 1 1>
  }
  n3 = Clamp(None=Round(None=@n2))
  n4 = Multiply(A=ScalarParameter() {group=Default, parameterName=GradMin_Default}, B=@n3)
  n5 = Multiply(A=ScalarParameter() {defaultValue=2, group=Default, parameterName=GradMax_Default}, B=@n3)
  n6 = TextureObjectParameter() {
   group=Default
   parameterName=ColorGradient_Standard
   texture=/FigmentEnvironment/LM_8_1_1/Athena/Environments/Landscape/Foliage/T_Athena_FlowerGrad_01
  }
  n7 = VectorParameter() {group=Default, parameterName=CenterColor_Standard, defaultValue=<rgba 0.67 0.3078 0.1381 0>}
  n8 = ScalarParameter() {defaultValue=0.75, group=Default, parameterName=BaseBlend_Scale}
  n9 = Multiply(
   A=LinearInterpolate(
    A=MaterialFunctionCall!(GradientMin=@n4, GradientMax=@n5, Grayscale=@n1.G, GradientTexture=@n6) {materialFunction=&/MF_ColorGradientMap}.Result
    B=Multiply(A=@n1.G, B=@n7.RGB)
    Alpha=@n1.B
   )
   B=LinearInterpolate(A=@n8, Alpha=@n1.G) {constA=0.1}
  )
  n10 = ScalarParameter() {group=Farm, parameterName=GradMin_Farm}
  n11 = ScalarParameter() {defaultValue=1, group=Farm, parameterName=GradMax_Farm}
  n12 = TextureObjectParameter() {
   group=Farm
   parameterName=ColorGradient_Farm
   texture=/FigmentEnvironment/LM_8_1_1/Athena/Environments/Landscape/Foliage/T_Athena_FlowerGrad_03
  }
  n13 = LinearInterpolate(
   A=MaterialFunctionCall(GradientMin=@n10, GradientMax=@n11, Grayscale=Multiply(A=@n1.G, B=@n3), GradientTexture=@n12) {materialFunction=&MF_ColorGradientMap}.Result
   B=Multiply(A=@n1.G, B=VectorParameter() {group=Farm, parameterName=CenterColor_Farm, defaultValue=<rgba 0.67 0.3078 0.1381 0>}.RGB)
   Alpha=@n1.B
  )
  n14 = LinearInterpolate(A=@n8, Alpha=@n1.G) {constA=0.1}
  n15 = MaterialFunctionCall() {materialFunction=&/MF_Athena_BiomeMask}
  n16 = TextureObjectParameter() {
   group=Swamp
   parameterName=ColorGradient_Swamp
   texture=/FigmentEnvironment/LM_8_1_1/Athena/Environments/Landscape/Foliage/T_Athena_FlowerGrad_04
  }
  n17 = MaterialFunctionCall(
   GradientMin=Multiply(A=ScalarParameter() {group=Swamp, parameterName=GradMin_Swamp}, B=@n3)
   GradientMax=Multiply(A=ScalarParameter() {defaultValue=2, group=Swamp, parameterName=GradMax_Swamp}, B=@n3)
   Grayscale=@n1.G
   GradientTexture=@n16
  ) {materialFunction=&MF_ColorGradientMap}
  n18 = Multiply(A=@n1.G, B=VectorParameter() {group=Swamp, parameterName=CenterColor_Swamp, defaultValue=<rgba 0.67 0.3078 0.1381 0>}.RGB)
  n19 = LinearInterpolate(
   A=LinearInterpolate(A=@n9, B=Multiply(A=@n13, B=@n14), Alpha=Round(None=@n15.Farm))
   B=Multiply(A=LinearInterpolate(A=@n17.Result, B=@n18, Alpha=@n1.B), B=LinearInterpolate(A=@n8, Alpha=@n1.G) {constA=0.1})
   Alpha=Round(None=@n15.Swamp)
  )
  n20 = Multiply(A=ScalarParameter() {group=Mountain, parameterName=GradMin_Mountain}, B=@n3)
  n21 = Multiply(A=ScalarParameter() {defaultValue=2, group=Mountain, parameterName=GradMax_Mountain}, B=@n3)
  n22 = TextureObjectParameter() {
   group=Mountain
   parameterName=ColorGradient_Mountain
   texture=/FigmentEnvironment/LM_8_1_1/Athena/Environments/Landscape/Foliage/T_Athena_FlowerGrad_02
  }
  n23 = VectorParameter() {group=Mountain, parameterName=CenterColor_Mountain, defaultValue=<rgba 0.67 0.3078 0.1381 0>}
  n24 = Multiply(
   A=LinearInterpolate(
    A=MaterialFunctionCall(GradientMin=@n20, GradientMax=@n21, Grayscale=@n1.G, GradientTexture=@n22) {materialFunction=&MF_ColorGradientMap}.Result
    B=Multiply(A=@n1.G, B=@n23.RGB)
    Alpha=@n1.B
   )
   B=LinearInterpolate(A=@n8, Alpha=@n1.G) {constA=0.1}
  )
  n25 = Round(None=@n15.Mountain)
  n26 = FunctionInput() {
   description=Rounded to nearest 0-1. Best used with RandomValuePerElement from PerObjectFoliageData to Lerp between 2 different color sets.
   inputName=Instance_RandomAlpha
   inputType=FunctionInput_Scalar
   sortPriority=1
  }
  n27 = StaticSwitchParameter(
   True=Multiply(A=@n1.G, B=VectorParameter() {parameterName=Flower_BaseColor, defaultValue=<rgba 1 1 1 0>}.RGB)
   False=LinearInterpolate(A=@n9, B=LinearInterpolate(A=@n19, B=@n24, Alpha=@n25), Alpha=@n26)
  ) {parameterName=Force_BaseColor}
  n28 = ScalarParameter() {defaultValue=1, parameterName=Flower_Brightness}
  n29 = MaterialFunctionCall() {materialFunction=&/MF_Athena_TerrainColorMatch}
  n30 = Multiply(
   A=MaterialFunctionCall!(Water, GrassColor=@n29.GrassColor) {materialFunction=/FigmentEnvironment/LM_8_1_1/Athena/Environments/Landscape/MaterialFunctions/Standard/MF_Athena_LakeGrass_01}.Grass_Color
   B=LinearInterpolate(Alpha=@n1.G) {constA=0.5}
  )
  n31 = Multiply(A=Multiply(A=@n29.Result, B=ScalarParameter() {defaultValue=1, parameterName=Stalk_Brightness}) {constB=0.95}, B=@n1.G) {constB=0.95}
  FunctionOutput(None=LinearInterpolate(A=Multiply(A=@n27, B=@n28), B=StaticSwitchParameter(True=@n30, False=@n31) {parameterName=Stalk_UseLakeColor}, Alpha=@n1.R))
  FunctionOutput(None=@n1.A) {outputName=Alpha, sortPriority=1}
  Multiply(A=@n11, B=@n3)
  Multiply(A=@n10, B=@n3)
  TextureSampleParameter2D() {
   parameterName=Flower_Masks
   samplerType=SAMPLERTYPE_LinearColor
   texture=&T_Athena_Flowers_01
  }
 ]}
