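# Landscape material function that tints the base color of incoming material attributes along a river, presumably on river spline or mesh UVs.
# The mask is a parabola on the V texture coordinate, scaled and raised to a high power into a soft band peaking mid-V.
# RiverTintAdd from MPC_Landscape supplies the color, and its alpha scales the mask.
# MatLayerBlend_OverrideBaseColor applies the clamped result, so despite the name it overrides rather than adds.
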MaterialFunction /FigmentEnvironment/LM_8_1_1/Athena/Environments/Landscape/MaterialFunctions/MF_River_Tint_Add
graph {nodes [
  n1 = CollectionParameter() {collection=/FigmentEnvironment/LM_8_1_1/Athena/Environments/Landscape/MPC/MPC_Landscape, parameterName=RiverTintAdd}
  n2 = ComponentMask(None=TextureCoordinate()) {g=true}
  n3 = Multiply(A=@n2, B=OneMinus(None=@n2))
  n4 = Multiply(A=Power(Base=Multiply(A=@n3) {constB=3.2}) {constExponent=10}, B=ComponentMask(None=@n1) {a=true})
  FunctionOutput(None=MaterialFunctionCall!(Material=FunctionInput() {inputType=FunctionInput_MaterialAttributes}, NewBaseColor=@n1, Mask=Clamp(None=@n4)) {materialFunction=/Engine/Functions/MaterialLayerFunctions/MatLayerBlend_OverrideBaseColor}.Blended Material)
 ]}
