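# Material function that outputs a storm mask, marking areas outside the Athena safe zone or near the storm vortex.
# Flooring saturated XY distance over safe-zone radius gives a hard circle edge, maxed with a linear falloff around the vortex center.
# Centers and safe-zone radius come from FortniteMaterialParameters; a static switch samples world or object position.
# Leftover branches, including a SphereMask variant and a wind-strength override, do not reach the output.
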MaterialFunction /FigmentEnvironment/LM_8_1_1/Athena/Environments/Landscape/MaterialFunctions/MF_StormMask
graph {nodes [
  n1 = CollectionParameter() {
   collection=/FigmentEnvironment/LM_8_1_1/Packages/Fortress_SharedMaterials/GlobalMaterialParameters/FortniteMaterialParameters
   parameterName=AthenaSafeZoneCenter
  }
  n2 = Distance(A=ComponentMask(None=WorldPosition().XYZ) {g=true, r=true}, B=ComponentMask(None=@n1) {g=true, r=true})
  n3 = CollectionParameter() {
   collection=&FortniteMaterialParameters
   parameterName=AthenaSafeZoneRadius
  }
  Floor(None=Saturate(None=Divide(A=@n2, B=@n3)))
  n4 = CollectionParameter() {
   collection=&FortniteMaterialParameters
   parameterName=AthenaSafeZoneCenter
  }
  n5 = ComponentMask(None=@n4) {g=true, r=true}
  n6 = ComponentMask(None=StaticSwitchParameter(True=WorldPosition().XYZ, False=ObjectPositionWS()) {parameterName=Use WorldPos instead of object}) {g=true, r=true}
  n7 = CollectionParameter() {
   collection=&FortniteMaterialParameters
   parameterName=AthenaSafeZoneRadius
  }
  n8 = Saturate(None=Divide(A=Distance(A=@n6, B=@n5), B=@n7))
  n9 = CollectionParameter() {
   collection=&FortniteMaterialParameters
   parameterName=AthenaVortexCenter
  }
  n10 = Distance(A=@n6, B=ComponentMask(None=@n9) {g=true, r=true})
  n11 = ScalarParameter() {defaultValue=30000, parameterName=AthenaVortexWidth}
  n12 = OneMinus(None=Saturate(None=Divide(A=@n10, B=@n11)))
  FunctionOutput(None=Max(A=Floor(None=@n8), B=@n12))
  Max(
   A=@n12
   B=CollectionParameter() {collection=/FigmentEnvironment/LM_8_1_1/Athena/Environments/Landscape/MPC/MPC_Landscape, parameterName=StormWindStrengthOverride}
  )
  n13 = WorldPosition()
  n14 = CollectionParameter() {
   collection=&FortniteMaterialParameters
   parameterName=AthenaSafeZoneCenter
  }
  n15 = CollectionParameter() {
   collection=&FortniteMaterialParameters
   parameterName=AthenaSafeZoneRadius
  }
  OneMinus(None=SphereMask(A=ComponentMask(None=@n13.XYZ) {g=true, r=true}, B=ComponentMask(None=@n14) {g=true, r=true}, Radius=@n15, Hardness=Constant() {r=1}))
 ]}
