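# MF_Athena_WorldHeightGrad is a landscape material function that turns world height into a linear gradient value.
# It takes the Z of the Coords input, defaulting to world position, subtracts a height offset and divides by a gradient scale, without clamping.
# Offset and scale default to MPC_Landscape collection parameters but can be overridden through inputs.
# Two scalar parameters with the same names sit unconnected in the graph.
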MaterialFunction /FigmentEnvironment/LM_8_1_1/Athena/Environments/Landscape/MaterialFunctions/MF_Athena_WorldHeightGrad
graph {nodes [
  n1 = ComponentMask(None=FunctionInput(Preview=WorldPosition().XYZ) {bUsePreviewValueAsDefault=true, inputName=Coords}) {b=true}
  n2 = FunctionInput(Preview=CollectionParameter() {collection=/FigmentEnvironment/LM_8_1_1/Athena/Environments/Landscape/MPC/MPC_Landscape, parameterName=HeightGrad_Offset}) {bUsePreviewValueAsDefault=true, inputName=height_Offset, inputType=FunctionInput_Scalar}
  n3 = CollectionParameter() {collection=&MPC_Landscape, parameterName=HeightGrad_Scale}
  FunctionOutput(None=Divide(
    A=Subtract(A=@n1, B=@n2)
    B=FunctionInput(Preview=@n3) {bUsePreviewValueAsDefault=true, inputName=height_GradScale, inputType=FunctionInput_Scalar}
   ))
  ScalarParameter() {defaultValue=15000, group=Height_Gradient, parameterName=HeightGrad_Scale}
  ScalarParameter() {defaultValue=-8000, group=Height_Gradient, parameterName=HeightGrad_Offset}
 ]}
