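# MF_TerrainDistanceFade is a material function that outputs a camera-distance fade factor, apparently for terrain materials.
# It measures world-position distance to the camera, subtracts a fixed near offset, divides by the Distance input, then clamps.
# With default clamp bounds, the result is 0 near the camera and rises linearly to 1 far away.
# The Distance input sets the fade length and falls back to its preview constant when unconnected.
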MaterialFunction /FigmentEnvironment/LM_8_1_1/Athena/Environments/Landscape/MaterialFunctions/MF_TerrainDistanceFade
graph {nodes [
  n1 = Subtract(A=Distance(A=WorldPosition().XYZ, B=CameraPositionWS()), B=Constant() {r=5000})
  n2 = FunctionInput(Preview=Constant() {r=55000}) {bUsePreviewValueAsDefault=true, inputName=Distance, inputType=FunctionInput_Scalar}
  FunctionOutput(None=Clamp(None=Divide(A=@n1, B=@n2)))
 ]}
